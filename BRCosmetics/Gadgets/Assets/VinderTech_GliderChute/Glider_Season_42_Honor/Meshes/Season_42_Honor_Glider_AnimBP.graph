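# Animation Blueprint for the Season 42 Honor umbrella glider, driving closed, opening, force-open and idle states, with front-end variants for cosmetic preview.
# An additive top-spin layers over the state machine, its rate following forward acceleration and zeroed in preview.
# Bone modifiers then adjust the crown and four cube bones, and spring-based AnimDynamics simulates the cubes.
# The event graph also computes time-based crown yaw and sine bobbing offsets, and flags full deployment on an anim notify.
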AnimBlueprint /BRCosmetics/Gadgets/Assets/VinderTech_GliderChute/Glider_Season_42_Honor/Meshes/Season_42_Honor_Glider_AnimBP
target_skeleton=&/Season_42_Honor_Glider_Skeleton
graphs [
 AnimGraph {nodes [
   Local To Component(LocalPose=Input Pose())
   AnimGraphNode_StateMachine_1 = Parachute_GliderItem() {is=title AnimGraphNode_StateMachine}
   n1 = Apply Additive(
    Base=@AnimGraphNode_StateMachine_1
    Additive="Sequence Player (additive) 'Umbrella_TopSpin_GLIDER'"!(PlayRate:float=Get TopSpin_Speed() -> (TopSpin_Speed:double)) -> (Pose) {sequence=/BRCosmetics/Animation/Game/MainPlayer/Skydive/ParaGlide/Umbrella/Umbrella_TopSpin_GLIDER}
   ) {alphaBoolBlend {blendOption=Linear}}
   n2 = "Constraint - dyn_Crown (S)"(ComponentPose=Local To Component(LocalPose=Slot 'DefaultSlot'(Source=@n1)), Alpha=1, ConstraintWeights_0=1) {boneToModify {boneName=dyn_Crown}, constraintSetup [{transformType=Scale, targetBone {boneName=blade_01}}], constraintWeights [1]}
   n3 = "Transform (Modify) Bone - Bone: dyn_Cube_B"!(
    ComponentPose="Transform (Modify) Bone - Bone: dyn_Cube_A"!(ComponentPose=@n2, Alpha=1, Translation="0,0,0") -> (Pose) {translationMode=BMM_Additive, boneToModify {boneName=dyn_Cube_A}}
    Alpha=1
    Translation="0,0,0"
   ) -> (Pose) {translationMode=BMM_Additive, translationSpace=BCS_BoneSpace, boneToModify {boneName=dyn_Cube_B}}
   n4 = "Transform (Modify) Bone - Bone: dyn_Crown"!(
    ComponentPose="Transform (Modify) Bone - Bone: dyn_Cube_D"!(
     ComponentPose="Transform (Modify) Bone - Bone: dyn_Cube_C"!(ComponentPose=@n3, Alpha=1, Translation="0,0,0") -> (Pose) {translationMode=BMM_Additive, translationSpace=BCS_BoneSpace, boneToModify {boneName=dyn_Cube_C}}
     Alpha=1
     Translation="0,0,0"
    ) -> (Pose) {translationMode=BMM_Additive, translationSpace=BCS_BoneSpace, boneToModify {boneName=dyn_Cube_D}}
    Alpha=1
    Translation="0,0,0"
    Rotation="0,50,0"
   ) -> (Pose) {
    rotationMode=BMM_Additive
    translationMode=BMM_Additive
    translationSpace=BCS_BoneSpace
    boneToModify {boneName=dyn_Crown}
    rotation {yaw=50}
   }
   n5 = Anim Dynamics - Bone: dyn_Cube_A(ComponentPose=@n4, Alpha=1, ExternalForce="0,0,0") {
    angularDampingOverride=1
    angularSpringConstant=10000
    bLinearSpring=true
    bOverrideLinearDamping=true
    bUseGravityOverride=true
    gravityScale=0.35
    linearDampingOverride=1
    linearSpringConstant=200
    simulationSpace=RootRelative
    boundBone {boneName=dyn_Cube_A}
    componentAppliedLinearAccClamp=<1000 1000 1000>
    componentLinearAccScale=<1 1 1>
    componentLinearVelScale=<1 1 1>
    gravityOverride {z=-500}
    physicsBodyDefinitions [{
      boundBone {boneName=dyn_Cube_A}
      boxExtents=<200 200 200>
      constraintSetup {angularLimitsMax=<0 0 0>, angularLimitsMin=<0 0 0>, angularTarget=<1 0 0>, linearAxesMax=<1.5 1.5 3>, linearAxesMin=<-1.5 -1.5 0>}
      localJointOffset=<2 0 0>
     }]
    simSpaceSettings {maxAngularAcceleration=30, maxAngularVelocity=6, simSpaceAngularAlpha=1}
   }
   n6 = Anim Dynamics - Bone: dyn_Cube_B(ComponentPose=@n5, Alpha=1, ExternalForce="0,0,0") {
    angularDampingOverride=1
    angularSpringConstant=10000
    bLinearSpring=true
    bOverrideLinearDamping=true
    bUseGravityOverride=true
    gravityScale=0.35
    linearDampingOverride=1
    linearSpringConstant=200
    simulationSpace=RootRelative
    alphaScaleBiasClamp {scale=0.8}
    boundBone {boneName=dyn_Cube_B}
    componentAppliedLinearAccClamp=<1000 1000 1000>
    componentLinearAccScale=<1 1 1>
    componentLinearVelScale=<1 1 1>
    gravityOverride {z=-500}
    physicsBodyDefinitions [{
      boundBone {boneName=dyn_Cube_B}
      boxExtents=<200 200 200>
      constraintSetup {angularLimitsMax=<0 0 0>, angularLimitsMin=<0 0 0>, angularTarget=<1 0 0>, linearAxesMax=<1 1 3>, linearAxesMin=<-1 -1 0>}
      localJointOffset=<2 0 0>
     }]
    simSpaceSettings {maxAngularAcceleration=30, maxAngularVelocity=6, simSpaceAngularAlpha=1}
   }
   n7 = Anim Dynamics - Bone: dyn_Cube_C(ComponentPose=@n6, Alpha=1, ExternalForce="0,0,0") {
    angularDampingOverride=1
    angularSpringConstant=10000
    bLinearSpring=true
    bOverrideLinearDamping=true
    bUseGravityOverride=true
    gravityScale=0.35
    linearDampingOverride=1
    linearSpringConstant=200
    simulationSpace=RootRelative
    boundBone {boneName=dyn_Cube_C}
    componentAppliedLinearAccClamp=<1000 1000 1000>
    componentLinearAccScale=<1 1 1>
    componentLinearVelScale=<1 1 1>
    gravityOverride {z=-500}
    physicsBodyDefinitions [{
      boundBone {boneName=dyn_Cube_C}
      boxExtents=<200 200 200>
      constraintSetup {angularLimitsMax=<0 0 0>, angularLimitsMin=<0 0 0>, angularTarget=<1 0 0>, linearAxesMax=<1.5 1.5 3>, linearAxesMin=<-1.5 -1.5 0>}
      localJointOffset=<2 0 0>
     }]
    simSpaceSettings {maxAngularAcceleration=30, maxAngularVelocity=6, simSpaceAngularAlpha=1}
   }
   n8 = Anim Dynamics - Bone: dyn_Cube_D(ComponentPose=@n7, Alpha=1, ExternalForce="0,0,0") {
    angularDampingOverride=1
    angularSpringConstant=10000
    bLinearSpring=true
    bOverrideLinearDamping=true
    bUseGravityOverride=true
    gravityScale=0.35
    linearDampingOverride=1
    linearSpringConstant=200
    simulationSpace=RootRelative
    alphaScaleBiasClamp {scale=0.8}
    boundBone {boneName=dyn_Cube_D}
    componentAppliedLinearAccClamp=<1000 1000 1000>
    componentLinearAccScale=<1 1 1>
    componentLinearVelScale=<1 1 1>
    gravityOverride {z=-500}
    physicsBodyDefinitions [{
      boundBone {boneName=dyn_Cube_D}
      boxExtents=<200 200 200>
      constraintSetup {angularLimitsMax=<0 0 0>, angularLimitsMin=<0 0 0>, angularTarget=<1 0 0>, linearAxesMax=<1 1 3>, linearAxesMin=<-1 -1 0>}
      localJointOffset=<2 0 0>
     }]
    simSpaceSettings {maxAngularAcceleration=30, maxAngularVelocity=6, simSpaceAngularAlpha=1}
   }
   Output Pose(Result=Component To Local(ComponentPose=@n8)) {layerGroup=None}
  ]}
 Parachute_GliderItem {
  entry=Closed
  path=AnimGraph.AnimGraphNode_StateMachine_1.Parachute_GliderItem
  ids=id
  nodes [
   AnimStateNode_0 = Parachute_Into() {is=title AnimStateNode}
   AnimStateNode_1 = Idle()
   AnimStateNode_2 = Closed() {is=title AnimStateNode}
   AnimStateNode_3 = ForceOpen_or_GliderItem() {is=title AnimStateNode}
   AnimStateTransitionNode_0 = AnimStateTransitionNode() {bAutomaticRuleBasedOnSequencePlayerInState=true, crossfadeDuration=0}
   AnimStateTransitionNode_1 = AnimStateTransitionNode()
   AnimStateTransitionNode_2 = AnimStateTransitionNode() {crossfadeDuration=0}
   AnimStateTransitionNode_3 = AnimStateTransitionNode() {crossfadeDuration=0}
   AnimStateTransitionNode_4 = AnimStateTransitionNode() {crossfadeDuration=0}
   AnimStateTransitionNode_5 = AnimStateTransitionNode()
   AnimStateTransitionNode_6 = AnimStateTransitionNode() {bAutomaticRuleBasedOnSequencePlayerInState=true, crossfadeDuration=0}
  ]
  links [
   {
    dst=ForceOpen_or_GliderItem
    rule=AnimGraph.AnimGraphNode_StateMachine_1.Parachute_GliderItem.AnimStateTransitionNode_4.Transition
    src=Closed
    via=AnimStateTransitionNode_4
   }
   {
    dst=Parachute_Into
    rule=AnimGraph.AnimGraphNode_StateMachine_1.Parachute_GliderItem.AnimStateTransitionNode_1.Transition
    src=Closed
    via=AnimStateTransitionNode_1
   }
   {
    dst=Closed
    rule=AnimGraph.AnimGraphNode_StateMachine_1.Parachute_GliderItem.AnimStateTransitionNode_5.Transition
    src=ForceOpen_or_GliderItem
    via=AnimStateTransitionNode_5
   }
   {
    dst=Idle
    rule=AnimGraph.AnimGraphNode_StateMachine_1.Parachute_GliderItem.AnimStateTransitionNode_6.Transition
    src=ForceOpen_or_GliderItem
    via=AnimStateTransitionNode_6
   }
   {
    dst=Closed
    rule=AnimGraph.AnimGraphNode_StateMachine_1.Parachute_GliderItem.AnimStateTransitionNode_2.Transition
    src=Idle
    via=AnimStateTransitionNode_2
   }
   {
    dst=Closed
    rule=AnimGraph.AnimGraphNode_StateMachine_1.Parachute_GliderItem.AnimStateTransitionNode_3.Transition
    src=Parachute_Into
    via=AnimStateTransitionNode_3
   }
   {
    dst=Idle
    rule=AnimGraph.AnimGraphNode_StateMachine_1.Parachute_GliderItem.AnimStateTransitionNode_0.Transition
    src=Parachute_Into
    via=AnimStateTransitionNode_0
   }
  ]
 }
 Parachute_Into {
  path=AnimGraph.AnimGraphNode_StateMachine_1.Parachute_GliderItem.AnimStateNode_0.Parachute_Into
  nodes [
   n1 = Get OwnerGlider_0() -> (OwnerGlider_0:Fort Player Parachute Object Reference)
   Output Animation Pose(Result=Blend Poses by bool(
     BlendPose_0=Sequence Player 'Umbrella_Into_FrontEnd_GLIDER'() {bLoopAnimation=false, sequence=/BRCosmetics/Animation/Game/MainPlayer/Skydive/ParaGlide/Umbrella/Umbrella_Into_FrontEnd_GLIDER}
     BlendPose_1=Sequence Player 'Umbrella_Into_GLIDER'!(PlayRate:float=1) -> (Pose) {bLoopAnimation=false, sequence=/BRCosmetics/Animation/Game/MainPlayer/Skydive/ParaGlide/Umbrella/Umbrella_Into_GLIDER}
     bActiveValue=Get IsCosmeticPreview!(self:Fort Player Parachute Object Reference=@n1) -> (IsCosmeticPreview:bool)
    ) {blendType=Linear})
  ]
 }
 Idle {
  path=AnimGraph.AnimGraphNode_StateMachine_1.Parachute_GliderItem.AnimStateNode_1.Idle
  nodes [
   n1 = Get OwnerGlider_0()
   Output Animation Pose(Result=Blend Poses by bool(
     BlendPose_0=Sequence Player 'Umbrella_Glide_FrontEnd_GLIDER'() {sequence=/BRCosmetics/Animation/Game/MainPlayer/Skydive/ParaGlide/Umbrella/Umbrella_Glide_FrontEnd_GLIDER}
     BlendPose_1=Sequence Player 'Umbrella_GLIDER'!(PlayRate:float=1) -> (Pose) {sequence=/Game/Animation/Game/MainPlayer/Skydive/ParaGlide/Umbrella/Umbrella_GLIDER}
     bActiveValue=Get IsCosmeticPreview(self=@n1)
    ) {blendType=Linear})
  ]
 }
 Closed {
  path=AnimGraph.AnimGraphNode_StateMachine_1.Parachute_GliderItem.AnimStateNode_2.Closed
  nodes [
   n1 = Get OwnerGlider_0()
   Output Animation Pose(Result=Blend Poses by bool(
     BlendPose_0=Sequence Player 'Umbrella_Closed_FrontEnd_GLIDER'() {sequence=/BRCosmetics/Animation/Game/MainPlayer/Skydive/ParaGlide/Umbrella/Umbrella_Closed_FrontEnd_GLIDER}
     BlendPose_1=Sequence Player 'Umbrella_Closed_GLIDER'() {sequence=/BRCosmetics/Animation/Game/MainPlayer/Skydive/ParaGlide/Umbrella/Umbrella_Closed_GLIDER}
     bActiveValue=Get IsCosmeticPreview(self=@n1)
    ) {blendType=Linear})
  ]
 }
 ForceOpen_or_GliderItem {
  path=AnimGraph.AnimGraphNode_StateMachine_1.Parachute_GliderItem.AnimStateNode_3.ForceOpen_or_GliderItem
  nodes [Output Animation Pose(Result=Sequence Player 'Umbrella_Into_GLIDER'(PlayRate=1) {bLoopAnimation=false, sequence=&Umbrella_Into_GLIDER})]
 }
 Transition {
  path=AnimGraph.AnimGraphNode_StateMachine_1.Parachute_GliderItem.AnimStateTransitionNode_0.Transition
  nodes [Result(bCanEnterTransition="Equal (Float)"!(A:double="Get Relevant Anim Time Remaining (Parachute_Into)"!() -> (ReturnValue:float) {is=title K2Node_AnimGetter}, B:double=0) -> (ReturnValue:bool) {is=title K2Node_CallFunction})]
 }
 Transition {
  path=AnimGraph.AnimGraphNode_StateMachine_1.Parachute_GliderItem.AnimStateTransitionNode_1.Transition
  nodes [
   n1 = Get SkydivingState() -> (SkydivingState:Fort Anim Input Skydiving Structure)
   Result(bCanEnterTransition=OR Boolean(
     A=Break Fort Anim Input Skydiving!(FortAnimInput_Skydiving:"Fort Anim Input Skydiving Structure (by ref)"=@n1) -> (
      ParaGilderRootOffset:Vector=~
      ParaGliderRootRotationOverride:Rotator=~
      bUseParaGlideRootModifier:bool=~
      bIsSkydivingFromLaunchPad:bool=~
      bIsSkydivingFromBus:bool=~
      bIsInVortex:bool=~
      bIsDiving:bool=~
      bIsParachuteOpen:bool=~
      bIsSkydiving:bool=~
      bIsParachuteLeaning:bool=~
      bIsSkydiveLeaning:bool=~
      bIsLeaning:bool=~
      bIsSkydiveDiveMode:bool=~
      bPlaySkydiveDrift:bool=~
      bSkydiveDriftDelayActive:bool=~
      bSkydiveDriftAnimAllowed:bool=~
      bIsGliderRight:bool=~
      bIsGliderCenter:bool=~
      bIsGliderLeft:bool=~
      bIsGliderForward:bool=~
      bIsGliderBack:bool=~
      bHasDeployedAndIsStandingUpright:bool=~
      LocalAccelForward:float=~
      LocalAccelRight:float=~
      LocalVelocityRight:float=~
      SkydiveAimPitch:float=~
      SkydiveAimYaw:float=~
      DeployChuteAnimRate:float=~
      SkydiveDriftAnimRate:float=~
      SkydiveDriftAnimRateCurrent:float=~
      SkydiveFidgetAnimRate:float=~
      SkydiveFidgetAnimRateCurrent:float=~
      SkydiveAdditiveAlpha:float=~
      SkydiveDriftDelay:float=~
      SkydiveDriftAnim:int
      SkydiveDriftAnimMax:int
      LaunchpadAnim:int
      LocalAccelDir:ESkydivingDirection Enum
      DirectionLast:ESkydivingDirection Enum
      PoseAdditive:Anim Sequence Object Reference
     ) {is=title K2Node_BreakStruct}.bIsParachuteOpen
     B=Get bForceOpen() -> (bForceOpen:bool)
    ))
   n2 = Break Fort Anim Input Skydiving(FortAnimInput_Skydiving=Get SkydivingState()) -> (
    ParaGilderRootOffset=~
    ParaGliderRootRotationOverride=~
    bUseParaGlideRootModifier=~
    bIsSkydivingFromLaunchPad=~
    bIsSkydivingFromBus=~
    bIsInVortex=~
    bIsDiving=~
    bIsParachuteOpen=~
    bIsSkydiving=~
    bIsParachuteLeaning=~
    bIsSkydiveLeaning=~
    bIsLeaning=~
    bIsSkydiveDiveMode=~
    bPlaySkydiveDrift=~
    bSkydiveDriftDelayActive=~
    bSkydiveDriftAnimAllowed=~
    bIsGliderRight=~
    bIsGliderCenter=~
    bIsGliderLeft=~
    bIsGliderForward=~
    bIsGliderBack=~
    bHasDeployedAndIsStandingUpright=~
    LocalAccelForward=~
    LocalAccelRight=~
    LocalVelocityRight=~
    SkydiveAimPitch=~
    SkydiveAimYaw=~
    DeployChuteAnimRate=~
    SkydiveDriftAnimRate=~
    SkydiveDriftAnimRateCurrent=~
    SkydiveFidgetAnimRate=~
    SkydiveFidgetAnimRateCurrent=~
    SkydiveAdditiveAlpha=~
    SkydiveDriftDelay=~
   )
   n3 = Get OwnerGlider_0()
   AND Boolean(
    A=OR Boolean!(A=@n2.bIsSkydivingFromLaunchPad, B=@n2.bIsSkydivingFromBus, C:bool=@n2.bIsInVortex, D:bool=Get IsCosmeticPreview(self=@n3)) -> (ReturnValue)
    B=@n2.bIsParachuteOpen
   )
  ]
 }
 Transition {
  path=AnimGraph.AnimGraphNode_StateMachine_1.Parachute_GliderItem.AnimStateTransitionNode_2.Transition
  nodes [Result(bCanEnterTransition=NOT Boolean(A=Get IsParachuteOpen() -> (IsParachuteOpen:bool)))]
 }
 Transition {
  path=AnimGraph.AnimGraphNode_StateMachine_1.Parachute_GliderItem.AnimStateTransitionNode_3.Transition
  nodes [
   n1 = Break Fort Anim Input Skydiving(FortAnimInput_Skydiving=Get SkydivingState()) -> (
    ParaGilderRootOffset=~
    ParaGliderRootRotationOverride=~
    bUseParaGlideRootModifier=~
    bIsSkydivingFromLaunchPad=~
    bIsSkydivingFromBus=~
    bIsInVortex=~
    bIsDiving=~
    bIsParachuteOpen=~
    bIsSkydiving=~
    bIsParachuteLeaning=~
    bIsSkydiveLeaning=~
    bIsLeaning=~
    bIsSkydiveDiveMode=~
    bPlaySkydiveDrift=~
    bSkydiveDriftDelayActive=~
    bSkydiveDriftAnimAllowed=~
    bIsGliderRight=~
    bIsGliderCenter=~
    bIsGliderLeft=~
    bIsGliderForward=~
    bIsGliderBack=~
    bHasDeployedAndIsStandingUpright=~
    LocalAccelForward=~
    LocalAccelRight=~
    LocalVelocityRight=~
    SkydiveAimPitch=~
    SkydiveAimYaw=~
    DeployChuteAnimRate=~
    SkydiveDriftAnimRate=~
    SkydiveDriftAnimRateCurrent=~
    SkydiveFidgetAnimRate=~
    SkydiveFidgetAnimRateCurrent=~
    SkydiveAdditiveAlpha=~
    SkydiveDriftDelay=~
   )
   n2 = Get bForceOpen()
   Result(bCanEnterTransition=AND Boolean(A=NOT Boolean(A=@n1.bIsParachuteOpen), B=NOT Boolean(A=@n2)))
  ]
 }
 Transition {
  path=AnimGraph.AnimGraphNode_StateMachine_1.Parachute_GliderItem.AnimStateTransitionNode_4.Transition
  nodes [Result(bCanEnterTransition=Get BlockAnimRule() -> (BlockAnimRule:bool))]
 }
 Transition {
  path=AnimGraph.AnimGraphNode_StateMachine_1.Parachute_GliderItem.AnimStateTransitionNode_5.Transition
  nodes [
   n1 = Break Fort Anim Input Skydiving(FortAnimInput_Skydiving=Get SkydivingState()) -> (
    ParaGilderRootOffset=~
    ParaGliderRootRotationOverride=~
    bUseParaGlideRootModifier=~
    bIsSkydivingFromLaunchPad=~
    bIsSkydivingFromBus=~
    bIsInVortex=~
    bIsDiving=~
    bIsParachuteOpen=~
    bIsSkydiving=~
    bIsParachuteLeaning=~
    bIsSkydiveLeaning=~
    bIsLeaning=~
    bIsSkydiveDiveMode=~
    bPlaySkydiveDrift=~
    bSkydiveDriftDelayActive=~
    bSkydiveDriftAnimAllowed=~
    bIsGliderRight=~
    bIsGliderCenter=~
    bIsGliderLeft=~
    bIsGliderForward=~
    bIsGliderBack=~
    bHasDeployedAndIsStandingUpright=~
    LocalAccelForward=~
    LocalAccelRight=~
    LocalVelocityRight=~
    SkydiveAimPitch=~
    SkydiveAimYaw=~
    DeployChuteAnimRate=~
    SkydiveDriftAnimRate=~
    SkydiveDriftAnimRateCurrent=~
    SkydiveFidgetAnimRate=~
    SkydiveFidgetAnimRateCurrent=~
    SkydiveAdditiveAlpha=~
    SkydiveDriftDelay=~
   )
   n2 = Get bForceOpen()
   Result(bCanEnterTransition=AND Boolean(A=NOT Boolean(A=@n1.bIsParachuteOpen), B=NOT Boolean(A=@n2)))
  ]
 }
 Transition {path=AnimGraph.AnimGraphNode_StateMachine_1.Parachute_GliderItem.AnimStateTransitionNode_6.Transition, nodes [Result()]}
 EventGraph {nodes [
   AnimNotify_TurnONEngines() {is=title K2Node_Event}
   > SetGliderFullyDeployed!(execute, self:Fort Glider Anim Instance Object Reference, bIsFullyDeployed:bool=true) -> (then)
   n1 = Event BlueprintUpdateAnimation()
   > n2 = Sequence()
   n3 = Break Fort Anim Input Skydiving(FortAnimInput_Skydiving=Get SkydivingState()) -> (
    ParaGilderRootOffset=~
    ParaGliderRootRotationOverride=~
    bUseParaGlideRootModifier=~
    bIsSkydivingFromLaunchPad=~
    bIsSkydivingFromBus=~
    bIsInVortex=~
    bIsDiving=~
    bIsParachuteOpen=~
    bIsSkydiving=~
    bIsParachuteLeaning=~
    bIsSkydiveLeaning=~
    bIsLeaning=~
    bIsSkydiveDiveMode=~
    bPlaySkydiveDrift=~
    bSkydiveDriftDelayActive=~
    bSkydiveDriftAnimAllowed=~
    bIsGliderRight=~
    bIsGliderCenter=~
    bIsGliderLeft=~
    bIsGliderForward=~
    bIsGliderBack=~
    bHasDeployedAndIsStandingUpright=~
    LocalAccelForward=~
    LocalAccelRight=~
    LocalVelocityRight=~
    SkydiveAimPitch=~
    SkydiveAimYaw=~
    DeployChuteAnimRate=~
    SkydiveDriftAnimRate=~
    SkydiveDriftAnimRateCurrent=~
    SkydiveFidgetAnimRate=~
    SkydiveFidgetAnimRateCurrent=~
    SkydiveAdditiveAlpha=~
    SkydiveDriftDelay=~
   )
   > Set IsParachuteOpen(IsParachuteOpen:bool=@n3.bIsParachuteOpen) -> (Output_Get:bool)
   > n4 = Branch(Condition=Get IsCosmeticPreview(self=Get OwnerGlider_0()))
   > Set TopSpin_Speed(TopSpin_Speed:double=0) -> (Output_Get:double)
   n5 = ForwardBack_Value_Interp!(
    execute=@n4.else
    Time:double=@n1.DeltaTimeX
    Currnet_Value:double=Get TopSpin_Speed()
    Idle_Value:double=1
    Idle_InterpSpeed:double=1
    Forward_Value:double=3
    Forward_InterpSpeed:double=2
    Back_Value:double=-1.75
    Back_InterpSpeed:double=2
    LocalAccelForward:double=@n3.LocalAccelForward
   ) -> (then, New_Value:double)
   > Set TopSpin_Speed(TopSpin_Speed=@n5)
   n6 = GetRealTimeSeconds()
   > Set CrownRotation(CrownRotation:Rotator=MakeRotator(Yaw=float * float(A=@n6, B=50) {is=title K2Node_PromotableOperator})) -> (Output_Get:Rotator)
   n7 = GetRealTimeSeconds()
   n8 = float * float(A="Sin (Radians)"(A=float * float(A=@n7, B=1.5) {is=title K2Node_CommutativeAssociativeBinaryOperator}), B=3)
   "Set Floater (slow)"(execute=@n2.then_1, "Floater (slow)":Vector=MakeVector(Y=0, Z=float + float(A=@n8, B=3))) -> (Output_Get:Vector)
   n9 = GetRealTimeSeconds()
   n10 = float * float(A="Sin (Radians)"(A=float * float(A=@n9, B=2) {is=title K2Node_CommutativeAssociativeBinaryOperator}), B=1.5) {is=title K2Node_PromotableOperator}
   > "Set Floater (lesser)"("Floater (lesser)":Vector=MakeVector(Y=0, Z=float + float(A=@n10, B=1.5))) -> (Output_Get:Vector)
   n11 = GetRealTimeSeconds()
   n12 = float * float(A="Sin (Radians)"(A=float * float(A=@n11, B=2.5) {is=title K2Node_CommutativeAssociativeBinaryOperator}), B=2) {is=title K2Node_PromotableOperator}
   > "Set Floater (more 1)"("Floater (more 1)":Vector=MakeVector(Y=0, Z=float + float(A=@n12, B=2))) -> (Output_Get:Vector)
   n13 = GetRealTimeSeconds()
   n14 = float * float(A="Sin (Radians)"(A=float * float(A=@n13, B=3) {is=title K2Node_CommutativeAssociativeBinaryOperator}), B=2.5) {is=title K2Node_PromotableOperator}
   > "Set Floater (more 2)"("Floater (more 2)":Vector=MakeVector(Y=0, Z=float + float(A=@n14, B=2.5))) -> (Output_Get:Vector)
   n15 = GetRealTimeSeconds()
   n16 = float * float(A="Sin (Radians)"(A=float * float(A=@n15, B=3.5) {is=title K2Node_CommutativeAssociativeBinaryOperator}), B=3) {is=title K2Node_PromotableOperator}
   > "Set Floater (most)"("Floater (most)":Vector=MakeVector(Y=0, Z=float + float(A=@n16, B=3))) -> (Output_Get:Vector)
  ]}
]
variables [
 BlockAnimRule {type=boolean}
 CrownRotation {type=Rotator}
 "Floater (lesser)" {type=Vector}
 "Floater (more 1)" {type=Vector}
 "Floater (more 2)" {type=Vector}
 "Floater (most)" {type=Vector}
 "Floater (slow)" {type=Vector}
 IsParachuteOpen {type=boolean}
 OwnerGlider_0 {type=FortPlayerParachute, default=null}
 TopSpin_Speed {type=number}
]
